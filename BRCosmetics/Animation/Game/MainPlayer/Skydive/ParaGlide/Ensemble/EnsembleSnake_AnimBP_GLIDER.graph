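# Animation Blueprint for the Ensemble Snake glider, covering deployment, gliding and animated mouth movement.
# A Closed/Into/Glide state machine follows parachute-open or force-open state, using FrontEnd clips for cosmetic previews and a forward-acceleration blendspace for gliding.
# A notify-started mouth state machine is applied additively, with a gated random delay picking one of three weighted variants.
# Glider_Base stays hidden when closed and reappears shortly after deployment begins.
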AnimBlueprint /BRCosmetics/Animation/Game/MainPlayer/Skydive/ParaGlide/Ensemble/EnsembleSnake_AnimBP_GLIDER
target_skeleton=/BRCosmetics/Gadgets/Assets/VinderTech_GliderChute/Glider_Ensemble_Snake_Male/Meshes/Ensemble_Snake_Male_Glider_Skeleton
graphs [
 AnimGraph {nodes [
   AnimGraphNode_StateMachine_0 = Main() {is=title AnimGraphNode_StateMachine}
   AnimGraphNode_StateMachine_1 = MouthAdditive() {is=title AnimGraphNode_StateMachine, bSkipFirstUpdateTransition=false}
   Output Pose(Result=Apply Additive(Base=@AnimGraphNode_StateMachine_0, Additive=@AnimGraphNode_StateMachine_1) {alphaBoolBlend {blendOption=Linear}}) {layerGroup=None}
  ]}
 Main {
  entry=Closed
  path=AnimGraph.AnimGraphNode_StateMachine_0.Main
  ids=id
  nodes [
   AnimStateNode_0 = Into() {is=title AnimStateNode, stateEntered {guid=#1, notifyName=Into_Entered}}
   AnimStateNode_1 = Glide() {is=title AnimStateNode}
   AnimStateNode_2 = Closed() {is=title AnimStateNode, stateEntered {guid=#2, notifyName=Closed_Entered}}
   AnimStateTransitionNode_0 = AnimStateTransitionNode() {bAutomaticRuleBasedOnSequencePlayerInState=true, crossfadeDuration=0}
   AnimStateTransitionNode_1 = AnimStateTransitionNode() {crossfadeDuration=0}
   AnimStateTransitionNode_2 = AnimStateTransitionNode() {crossfadeDuration=0}
   AnimStateTransitionNode_3 = AnimStateTransitionNode() {crossfadeDuration=0}
  ]
  links [
   {dst=Into, rule=AnimGraph.AnimGraphNode_StateMachine_0.Main.AnimStateTransitionNode_1.Transition, src=Closed, via=AnimStateTransitionNode_1}
   {dst=Closed, rule=AnimGraph.AnimGraphNode_StateMachine_0.Main.AnimStateTransitionNode_3.Transition, src=Glide, via=AnimStateTransitionNode_3}
   {dst=Closed, rule=AnimGraph.AnimGraphNode_StateMachine_0.Main.AnimStateTransitionNode_2.Transition, src=Into, via=AnimStateTransitionNode_2}
   {dst=Glide, rule=AnimGraph.AnimGraphNode_StateMachine_0.Main.AnimStateTransitionNode_0.Transition, src=Into, via=AnimStateTransitionNode_0}
  ]
 }
 Into {
  path=AnimGraph.AnimGraphNode_StateMachine_0.Main.AnimStateNode_0.Into
  nodes [
   n1 = Get OwnerGlider() -> (OwnerGlider:Fort Player Parachute Object Reference)
   Output Animation Pose(Result=Blend Poses by bool(
     BlendPose_0=Sequence Player 'EnsembleSnake_Into_FrontEnd_GLIDER'() {bLoopAnimation=false, sequence=&/EnsembleSnake_Into_FrontEnd_GLIDER}
     BlendPose_1=Sequence Player 'EnsembleSnake_Into_GLIDER'!(bLoopAnimation:bool=False) -> (Pose) {bLoopAnimation=false, sequence=&/EnsembleSnake_Into_GLIDER}
     bActiveValue=Get IsCosmeticPreview!(self:Fort Player Parachute Object Reference=@n1) -> (IsCosmeticPreview:bool)
    ) {blendType=Linear})
  ]
 }
 Glide {
  path=AnimGraph.AnimGraphNode_StateMachine_0.Main.AnimStateNode_1.Glide
  nodes [
   n1 = Break Fort Anim Input Skydiving!(FortAnimInput_Skydiving:"Fort Anim Input Skydiving Structure (by ref)"=Get SkydivingState() -> (SkydivingState:Fort Anim Input Skydiving Structure)) -> (
    ParaGilderRootOffset:Vector=~
    ParaGliderRootRotationOverride:Rotator=~
    bUseParaGlideRootModifier:bool=~
    bIsSkydivingFromLaunchPad:bool=~
    bIsSkydivingFromBus:bool=~
    bIsInVortex:bool=~
    bIsDiving:bool=~
    bIsParachuteOpen:bool=~
    bIsSkydiving:bool=~
    bIsParachuteLeaning:bool=~
    bIsSkydiveLeaning:bool=~
    bIsLeaning:bool=~
    bIsSkydiveDiveMode:bool=~
    bPlaySkydiveDrift:bool=~
    bSkydiveDriftDelayActive:bool=~
    bSkydiveDriftAnimAllowed:bool=~
    bIsGliderRight:bool=~
    bIsGliderCenter:bool=~
    bIsGliderLeft:bool=~
    bIsGliderForward:bool=~
    bIsGliderBack:bool=~
    bHasDeployedAndIsStandingUpright:bool=~
    LocalAccelForward:float=~
    LocalAccelRight:float=~
    LocalVelocityRight:float=~
    SkydiveAimPitch:float=~
    SkydiveAimYaw:float=~
    DeployChuteAnimRate:float=~
    SkydiveDriftAnimRate:float=~
    SkydiveDriftAnimRateCurrent:float=~
    SkydiveFidgetAnimRate:float=~
    SkydiveFidgetAnimRateCurrent:float=~
    SkydiveAdditiveAlpha:float=~
    SkydiveDriftDelay:float=~
    SkydiveDriftAnim:int
    SkydiveDriftAnimMax:int
    LaunchpadAnim:int
    LocalAccelDir:ESkydivingDirection Enum
    DirectionLast:ESkydivingDirection Enum
    PoseAdditive:Anim Sequence Object Reference
   ) {is=title K2Node_BreakStruct}
   n2 = Get OwnerGlider()
   Output Animation Pose(Result=Blend Poses by bool(
     BlendPose_0=Sequence Player 'EnsembleSnake_Glide_FrontEnd_GLIDER'() {sequence=&/EnsembleSnake_Glide_FrontEnd_GLIDER}
     BlendPose_1=Blendspace Player 'BS_EnsembleSnake_Glide_GLIDER'(X=0, Y=@n1.LocalAccelForward) {blendSpace=&/BS_EnsembleSnake_Glide_GLIDER}
     bActiveValue=Get IsCosmeticPreview(self=@n2)
    ) {blendType=Linear})
  ]
 }
 Closed {
  path=AnimGraph.AnimGraphNode_StateMachine_0.Main.AnimStateNode_2.Closed
  nodes [
   n1 = Get OwnerGlider()
   Output Animation Pose(Result=Blend Poses by bool(
     BlendPose_0=Sequence Player 'EnsembleSnake_Closed_FrontEnd_GLIDER'() {sequence=&/EnsembleSnake_Closed_FrontEnd_GLIDER}
     BlendPose_1=Sequence Player 'EnsembleSnake_Closed_GLIDER'() {sequence=&/EnsembleSnake_Closed_GLIDER}
     bActiveValue=Get IsCosmeticPreview(self=@n1)
    ) {blendType=Linear})
  ]
 }
 Transition {path=AnimGraph.AnimGraphNode_StateMachine_0.Main.AnimStateTransitionNode_0.Transition, nodes [Result()]}
 Transition {
  path=AnimGraph.AnimGraphNode_StateMachine_0.Main.AnimStateTransitionNode_1.Transition
  nodes [
   n1 = Get SkydivingState()
   Result(bCanEnterTransition=OR Boolean(
     A=Break Fort Anim Input Skydiving(FortAnimInput_Skydiving=@n1) -> (
      ParaGilderRootOffset=~
      ParaGliderRootRotationOverride=~
      bUseParaGlideRootModifier=~
      bIsSkydivingFromLaunchPad=~
      bIsSkydivingFromBus=~
      bIsInVortex=~
      bIsDiving=~
      bIsParachuteOpen=~
      bIsSkydiving=~
      bIsParachuteLeaning=~
      bIsSkydiveLeaning=~
      bIsLeaning=~
      bIsSkydiveDiveMode=~
      bPlaySkydiveDrift=~
      bSkydiveDriftDelayActive=~
      bSkydiveDriftAnimAllowed=~
      bIsGliderRight=~
      bIsGliderCenter=~
      bIsGliderLeft=~
      bIsGliderForward=~
      bIsGliderBack=~
      bHasDeployedAndIsStandingUpright=~
      LocalAccelForward=~
      LocalAccelRight=~
      LocalVelocityRight=~
      SkydiveAimPitch=~
      SkydiveAimYaw=~
      DeployChuteAnimRate=~
      SkydiveDriftAnimRate=~
      SkydiveDriftAnimRateCurrent=~
      SkydiveFidgetAnimRate=~
      SkydiveFidgetAnimRateCurrent=~
      SkydiveAdditiveAlpha=~
      SkydiveDriftDelay=~
     ).bIsParachuteOpen
     B=Get bForceOpen() -> (bForceOpen:bool)
    ))
  ]
 }
 Transition {
  path=AnimGraph.AnimGraphNode_StateMachine_0.Main.AnimStateTransitionNode_2.Transition
  nodes [
   n1 = Get bForceOpen()
   n2 = Break Fort Anim Input Skydiving(FortAnimInput_Skydiving=Get SkydivingState()) -> (
    ParaGilderRootOffset=~
    ParaGliderRootRotationOverride=~
    bUseParaGlideRootModifier=~
    bIsSkydivingFromLaunchPad=~
    bIsSkydivingFromBus=~
    bIsInVortex=~
    bIsDiving=~
    bIsParachuteOpen=~
    bIsSkydiving=~
    bIsParachuteLeaning=~
    bIsSkydiveLeaning=~
    bIsLeaning=~
    bIsSkydiveDiveMode=~
    bPlaySkydiveDrift=~
    bSkydiveDriftDelayActive=~
    bSkydiveDriftAnimAllowed=~
    bIsGliderRight=~
    bIsGliderCenter=~
    bIsGliderLeft=~
    bIsGliderForward=~
    bIsGliderBack=~
    bHasDeployedAndIsStandingUpright=~
    LocalAccelForward=~
    LocalAccelRight=~
    LocalVelocityRight=~
    SkydiveAimPitch=~
    SkydiveAimYaw=~
    DeployChuteAnimRate=~
    SkydiveDriftAnimRate=~
    SkydiveDriftAnimRateCurrent=~
    SkydiveFidgetAnimRate=~
    SkydiveFidgetAnimRateCurrent=~
    SkydiveAdditiveAlpha=~
    SkydiveDriftDelay=~
   )
   Result(bCanEnterTransition=AND Boolean(A=NOT Boolean(A=@n2.bIsParachuteOpen), B=NOT Boolean(A=@n1)))
  ]
 }
 Transition {
  path=AnimGraph.AnimGraphNode_StateMachine_0.Main.AnimStateTransitionNode_3.Transition
  nodes [
   n1 = Get bForceOpen()
   n2 = Break Fort Anim Input Skydiving(FortAnimInput_Skydiving=Get SkydivingState()) -> (
    ParaGilderRootOffset=~
    ParaGliderRootRotationOverride=~
    bUseParaGlideRootModifier=~
    bIsSkydivingFromLaunchPad=~
    bIsSkydivingFromBus=~
    bIsInVortex=~
    bIsDiving=~
    bIsParachuteOpen=~
    bIsSkydiving=~
    bIsParachuteLeaning=~
    bIsSkydiveLeaning=~
    bIsLeaning=~
    bIsSkydiveDiveMode=~
    bPlaySkydiveDrift=~
    bSkydiveDriftDelayActive=~
    bSkydiveDriftAnimAllowed=~
    bIsGliderRight=~
    bIsGliderCenter=~
    bIsGliderLeft=~
    bIsGliderForward=~
    bIsGliderBack=~
    bHasDeployedAndIsStandingUpright=~
    LocalAccelForward=~
    LocalAccelRight=~
    LocalVelocityRight=~
    SkydiveAimPitch=~
    SkydiveAimYaw=~
    DeployChuteAnimRate=~
    SkydiveDriftAnimRate=~
    SkydiveDriftAnimRateCurrent=~
    SkydiveFidgetAnimRate=~
    SkydiveFidgetAnimRateCurrent=~
    SkydiveAdditiveAlpha=~
    SkydiveDriftDelay=~
   )
   Result(bCanEnterTransition=AND Boolean(A=NOT Boolean(A=@n2.bIsParachuteOpen), B=NOT Boolean(A=@n1)))
  ]
 }
 MouthAdditive {
  entry=Identity
  path=AnimGraph.AnimGraphNode_StateMachine_1.MouthAdditive
  ids=id
  nodes [
   AnimStateNode_0 = Identity() {is=title AnimStateNode}
   AnimStateNode_1 = MouthClosed() {
    is=title AnimStateNode
    stateEntered {guid=#3, notifyName=MouthClosed_Entered}
    stateFullyBlended {guid=#4, notifyName=MouthClosed_FullyBlended}
    stateLeft {guid=#5, notifyName=MouthClosed_Exit}
   }
   AnimStateNode_2 = MouthAnimC() {is=title AnimStateNode}
   AnimStateNode_3 = MouthAnimA() {is=title AnimStateNode}
   AnimStateNode_4 = MouthAnimB() {is=title AnimStateNode}
   AnimStateTransitionNode_0 = AnimStateTransitionNode() {crossfadeDuration=0.4}
   AnimStateTransitionNode_1 = AnimStateTransitionNode() {crossfadeDuration=0}
   AnimStateTransitionNode_2 = AnimStateTransitionNode() {crossfadeDuration=0}
   AnimStateTransitionNode_3 = AnimStateTransitionNode() {bAutomaticRuleBasedOnSequencePlayerInState=true, crossfadeDuration=0}
   AnimStateTransitionNode_4 = AnimStateTransitionNode() {bAutomaticRuleBasedOnSequencePlayerInState=true, crossfadeDuration=0}
   AnimStateTransitionNode_5 = AnimStateTransitionNode() {bAutomaticRuleBasedOnSequencePlayerInState=true, crossfadeDuration=0}
   AnimStateTransitionNode_6 = AnimStateTransitionNode() {crossfadeDuration=0}
   AnimStateTransitionNode_7 = AnimStateTransitionNode() {crossfadeDuration=0}
  ]
  links [
   {
    dst=MouthClosed
    rule=AnimGraph.AnimGraphNode_StateMachine_1.MouthAdditive.AnimStateTransitionNode_0.Transition
    src=Identity
    via=AnimStateTransitionNode_0
   }
   {
    dst=MouthClosed
    rule=AnimGraph.AnimGraphNode_StateMachine_1.MouthAdditive.AnimStateTransitionNode_3.Transition
    src=MouthAnimA
    via=AnimStateTransitionNode_3
   }
   {
    dst=MouthClosed
    rule=AnimGraph.AnimGraphNode_StateMachine_1.MouthAdditive.AnimStateTransitionNode_4.Transition
    src=MouthAnimB
    via=AnimStateTransitionNode_4
   }
   {
    dst=MouthClosed
    rule=AnimGraph.AnimGraphNode_StateMachine_1.MouthAdditive.AnimStateTransitionNode_5.Transition
    src=MouthAnimC
    via=AnimStateTransitionNode_5
   }
   {
    dst=Identity
    rule=AnimGraph.AnimGraphNode_StateMachine_1.MouthAdditive.AnimStateTransitionNode_1.Transition
    src=MouthClosed
    via=AnimStateTransitionNode_1
   }
   {
    dst=MouthAnimA
    rule=AnimGraph.AnimGraphNode_StateMachine_1.MouthAdditive.AnimStateTransitionNode_2.Transition
    src=MouthClosed
    via=AnimStateTransitionNode_2
   }
   {
    dst=MouthAnimB
    rule=AnimGraph.AnimGraphNode_StateMachine_1.MouthAdditive.AnimStateTransitionNode_6.Transition
    src=MouthClosed
    via=AnimStateTransitionNode_6
   }
   {
    dst=MouthAnimC
    rule=AnimGraph.AnimGraphNode_StateMachine_1.MouthAdditive.AnimStateTransitionNode_7.Transition
    src=MouthClosed
    via=AnimStateTransitionNode_7
   }
  ]
 }
 Identity {
  path=AnimGraph.AnimGraphNode_StateMachine_1.MouthAdditive.AnimStateNode_0.Identity
  nodes [Output Animation Pose(Result=Additive Identity Pose())]
 }
 MouthClosed {
  path=AnimGraph.AnimGraphNode_StateMachine_1.MouthAdditive.AnimStateNode_1.MouthClosed
  nodes [Output Animation Pose(Result="Sequence Player (additive) 'EnsembleSnake_MouthAdditive_ClosedPose_GLIDER'"() {sequence=&/EnsembleSnake_MouthAdditive_ClosedPose_GLIDER})]
 }
 MouthAnimC {
  path=AnimGraph.AnimGraphNode_StateMachine_1.MouthAdditive.AnimStateNode_2.MouthAnimC
  nodes [
   n1 = "Sequence Player (additive) 'EnsembleSnake_MouthAdditive_C_GLIDER'"!(bLoopAnimation:bool=False) -> (Pose) {bLoopAnimation=false, sequence=&/EnsembleSnake_MouthAdditive_C_GLIDER}
   Output Animation Pose(Result=@n1)
  ]
 }
 MouthAnimA {
  path=AnimGraph.AnimGraphNode_StateMachine_1.MouthAdditive.AnimStateNode_3.MouthAnimA
  nodes [
   n1 = "Sequence Player (additive) 'EnsembleSnake_MouthAdditive_A_GLIDER'"!(bLoopAnimation:bool=False) -> (Pose) {bLoopAnimation=false, sequence=&/EnsembleSnake_MouthAdditive_A_GLIDER}
   Output Animation Pose(Result=@n1)
  ]
 }
 MouthAnimB {
  path=AnimGraph.AnimGraphNode_StateMachine_1.MouthAdditive.AnimStateNode_4.MouthAnimB
  nodes [
   n1 = "Sequence Player (additive) 'EnsembleSnake_MouthAdditive_B_GLIDER'"!(bLoopAnimation:bool=False) -> (Pose) {bLoopAnimation=false, sequence=&/EnsembleSnake_MouthAdditive_B_GLIDER}
   Output Animation Pose(Result=@n1)
  ]
 }
 Transition {
  path=AnimGraph.AnimGraphNode_StateMachine_1.MouthAdditive.AnimStateTransitionNode_0.Transition
  nodes [Result(bCanEnterTransition=Get InitiateMouthAdditives() -> (InitiateMouthAdditives:bool))]
 }
 Transition {
  path=AnimGraph.AnimGraphNode_StateMachine_1.MouthAdditive.AnimStateTransitionNode_1.Transition
  nodes [Result(bCanEnterTransition=NOT Boolean(A=Get InitiateMouthAdditives()))]
 }
 Transition {
  path=AnimGraph.AnimGraphNode_StateMachine_1.MouthAdditive.AnimStateTransitionNode_2.Transition
  nodes [Result(bCanEnterTransition=Get PlayMouthAdditiveA() -> (PlayMouthAdditiveA:bool))]
 }
 Transition {path=AnimGraph.AnimGraphNode_StateMachine_1.MouthAdditive.AnimStateTransitionNode_3.Transition, nodes [Result()]}
 Transition {path=AnimGraph.AnimGraphNode_StateMachine_1.MouthAdditive.AnimStateTransitionNode_4.Transition, nodes [Result()]}
 Transition {path=AnimGraph.AnimGraphNode_StateMachine_1.MouthAdditive.AnimStateTransitionNode_5.Transition, nodes [Result()]}
 Transition {
  path=AnimGraph.AnimGraphNode_StateMachine_1.MouthAdditive.AnimStateTransitionNode_6.Transition
  nodes [Result(bCanEnterTransition=Get PlayMouthAdditiveB() -> (PlayMouthAdditiveB:bool))]
 }
 Transition {
  path=AnimGraph.AnimGraphNode_StateMachine_1.MouthAdditive.AnimStateTransitionNode_7.Transition
  nodes [Result(bCanEnterTransition=Get PlayMouthAdditiveC() -> (PlayMouthAdditiveC:bool))]
 }
 EventGraph {nodes [
   AnimNotify_TurnONEngines() {is=title K2Node_Event}
   > SetGliderFullyDeployed!(execute, self:Fort Glider Anim Instance Object Reference, bIsFullyDeployed:bool=true) -> (then)
   AnimNotify_InitiateMouthAdditives() {is=title K2Node_Event}
   > Set InitiateMouthAdditives(InitiateMouthAdditives:bool=true) -> (Output_Get:bool)
   AnimNotify_Closed_Entered() {is=title K2Node_Event}
   > n1 = Sequence()
   > Set InitiateMouthAdditives(InitiateMouthAdditives=false)
   > Set PlayMouthAdditiveA(PlayMouthAdditiveA:bool=false) -> (Output_Get:bool)
   > Set PlayMouthAdditiveB(PlayMouthAdditiveB:bool=false) -> (Output_Get:bool)
   > n2 = Set PlayMouthAdditiveC(PlayMouthAdditiveC:bool=false) -> (Output_Get:bool)
   Gate!(Enter:Exec=@n8.then, Open:Exec=@n7.then_0, Close:Exec=@n2.then, Toggle:Exec, bStartClosed:bool=true) -> (Exit:Exec) {is=title K2Node_MacroInstance}
   > n3 = Set PlayMouthAdditiveC(PlayMouthAdditiveC=RandomBoolWithWeight!(Weight:float=0.25) -> (ReturnValue:bool))
   > n4 = Branch(Condition=@n3)
   n5 = Branch(execute=@n4.else, Condition=RandomBool!() -> (ReturnValue:bool))
   > Set PlayMouthAdditiveA(PlayMouthAdditiveA=true)
   Set PlayMouthAdditiveB(execute=@n5.else, PlayMouthAdditiveB=true)
   n6 = GetOwningComponent()
   HideBoneByName!(
    execute=[@n9.then, @n1.then_1]
    self:Skinned Mesh Component Object Reference=@n6
    BoneName:Name=Glider_Base
    PhysBodyOption:EPhysBodyOp Enum=PBO_None
   ) -> (then)
   AnimNotify_MouthClosed_Exit() {is=title K2Node_Event}
   > Set PlayMouthAdditiveA(PlayMouthAdditiveA=false)
   > Set PlayMouthAdditiveB(PlayMouthAdditiveB=false)
   > Set PlayMouthAdditiveC(PlayMouthAdditiveC=false)
   AnimNotify_MouthClosed_FullyBlended() {is=title K2Node_Event}
   > n7 = Sequence()
   n8 = RetriggerableDelay(execute=@n7.then_1, Duration=RandomFloatInRange(Min=1, Max=3))
   n9 = Event BlueprintInitializeAnimation()
   AnimNotify_Into_Entered() {is=title K2Node_Event}
   > Delay(Duration=0.05)
   > UnHideBoneByName!(execute, self:Skinned Mesh Component Object Reference=@n6, BoneName:Name=Glider_Base) -> (then)
   TryGetPawnOwner()
   Event BlueprintUpdateAnimation()
   AnimNotify_MouthClosed_Entered() {is=title K2Node_Event}
  ]}
]
variables [
 InitiateMouthAdditives {type=boolean}
 PlayMouthAdditiveA {type=boolean}
 PlayMouthAdditiveB {type=boolean}
 PlayMouthAdditiveC {type=boolean}
]
